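# Editor utility widget that creates a harvester tool animation montage from a selected Anim Sequence.
# On construct, three Single Property Views bind to the widget itself, exposing its SourceAnimation, Destination and Name variables.
# Changing the animation derives Name from GetComboName plus a "_Harvesting_Combo_M" suffix, and Destination from the animation's folder if it exists.
# The button validates animation, directory and name, then calls CreateHarvesterMontage.
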EditorUtilityWidgetBlueprint /Game/ArtTools/Tech_Anim/HarvesterTool/HarvesterTool_EWidget
graphs [
 EventGraph {nodes [
   "On Property Changed (AnimationAsset)"!() -> (OutputDelegate, then, PropertyName:Name=None) {is=title K2Node_ComponentBoundEvent}
   > UpdateComboNameAndPath()
   "On Clicked (CreateMontageButton)"() {is=title K2Node_ComponentBoundEvent}
   > OnCreateClicked!(
    execute
    self
    SelectedAnim:Anim Sequence Object Reference=Get SourceAnimation() -> (SourceAnimation:Anim Sequence Object Reference)
    Path:String=Get Destination() -> (Destination:String)
    Name:String=Get Name() -> (Name:String)
   ) -> (then, Success:bool)
   Event PreConstruct()
   > SetDesiredSizeInViewport!(execute, self:User Widget Object Reference, Size:Vector 2D Structure=(X=200,Y=80)) -> (then)
   Event Construct()
   > SetDesiredSizeInViewport(Size=(X=200,Y=80))
   n1 = Get AnimationAsset() -> (AnimationAsset:Single Property View Object Reference)
   > SetEditorProperty(Object=@n1, PropertyName=PropertyName, PropertyValue:Anim Sequence Object Reference=Get SourceAnimation())
   n2 = Self-Reference()
   > SetObject(self=@n1, NewObject=@n2)
   > SetObject(self=Get OutputDirectory() -> (OutputDirectory:Single Property View Object Reference), NewObject=@n2)
   > SetObject(self=Get AssetName() -> (AssetName:Single Property View Object Reference), NewObject=@n2)
   > UpdateComboNameAndPath()
  ]}
 OnCreateClicked {nodes [
   n1 = OnCreateClicked!() -> (then, SelectedAnim:Anim Sequence Object Reference, Path:String, Name:String) {is=title K2Node_FunctionEntry}
   > n2 = Branch(Condition=IsValid(Object=@n1))
   > n3 = DoesDirectoryExist!(execute, DirectoryPath:String=@n1.Path) -> (then, ReturnValue:bool)
   > n4 = Branch(Condition=@n3)
   > n5 = Branch(Condition=IsEmpty(InString=@n1.Name))
   Return Node!(execute=[@n5.then, @n4.else, @n2.else], Success:bool=false)
   n6 = CreateHarvesterMontage!(execute=@n5.else, SourceAnimation:Anim Sequence Object Reference=@n1, Path:String=@n1.Path, Name:String=@n1.Name) -> (then, ReturnValue:Anim Montage Object Reference)
   > Return Node(Success=IsValid(Object=@n6))
  ]}
 UpdateComboNameAndPath {nodes [
   UpdateComboNameAndPath() {is=title K2Node_FunctionEntry}
   n1 = Get SourceAnimation()
   > Branch(Condition=IsValid(Object=@n1))
   > n2 = GetComboName!(execute, AnimSequence:Anim Sequence Object Reference=@n1) -> (then, ReturnValue:String)
   > Set Name(Name:String=Append(A=@n2, B=_Harvesting_Combo_M)) -> (Output_Get:String)
   n3 = GetPath!(InPath:String=Get Object Path String(Object=@n1)) -> (ReturnValue:String)
   > n4 = DoesDirectoryExist(DirectoryPath=@n3)
   > Branch(Condition=@n4)
   > Set Destination(Destination:String=@n3) -> (Output_Get:String)
  ]}
]
variables [
 Destination {type=string, default=""}
 Name {type=string, default=""}
 SourceAnimation {type=AnimSequence, default=/Game/Animation/Game/MainPlayer/Combat/Melee/TwoHanded/Staff/GimmickMale/GimmickMale_Crouch_Swing1}
]
widgets {
 info {rootWidgetClass=&CanvasPanel, widgetCount=9}
 widgets [
  CanvasPanel_0 {widgetClassPath=&CanvasPanel, widget=CanvasPanel_0}
  VerticalBox_0 {
   widgetClassPath=&VerticalBox
   parent=CanvasPanel_0
   widget=VerticalBox_0
   properties {clipping=ClipToBounds}
   slot {
    class=CanvasPanelSlot
    properties {bAutoSize=true, layoutData {alignment=<0 0>, anchors {maximum=<0.7711 0.7764>, minimum=<0 0>}, offsets=<ltrb 0 0 -439.9399 -241.7418>}}
   }
  }
  HorizontalBox_117 {widgetClassPath=&HorizontalBox, parent=VerticalBox_0, widget=HorizontalBox_117, slot {class=VerticalBoxSlot}}
  AnimationAsset {
   widgetClassPath=&SinglePropertyView
   parent=HorizontalBox_117
   widget=AnimationAsset
   properties {propertyName=SourceAnimation}
   slot {class=HorizontalBoxSlot, properties {padding=<ltrb 5 5 5 5>}}
  }
  OutputDirectory {
   widgetClassPath=&SinglePropertyView
   parent=VerticalBox_0
   widget=OutputDirectory
   properties {propertyName=Destination}
   slot {class=VerticalBoxSlot, properties {padding=<ltrb 5 5 5 5>}}
  }
  AssetName {
   widgetClassPath=&SinglePropertyView
   parent=VerticalBox_0
   widget=AssetName
   properties {propertyName=Name}
   slot {class=VerticalBoxSlot, properties {padding=<ltrb 5 5 5 5>}}
  }
  HorizontalBox_192 {
   widgetClassPath=&HorizontalBox
   parent=VerticalBox_0
   widget=HorizontalBox_192
   slot {class=VerticalBoxSlot, properties {padding=<ltrb 5 5 5 5>}}
  }
  CreateMontageButton {
   widgetClassPath=&Button
   parent=HorizontalBox_192
   widget=CreateMontageButton
   properties {widgetStyle {
     clickedSlateSound {resourceObject=null}
     disabled {
      bIsDynamicallyLoaded=false
      drawAs=RoundedBox
      imageType=NoImage
      mirroring=NoMirror
      resourceName=None
      resourceObject=null
      tiling=NoTile
      imageSize=<0 0>
      margin=<ltrb 0 0 0 0>
      outlineSettings {
       bUseBrushTransparency=false
       roundingType=FixedRadius
       width=1
       color {colorUseRule=UseColor_ColorTable, specifiedColor=<rgba 0 0 0 0>}
       cornerRadii=<4 4 4 4>
      }
      tintColor {colorUseRule=UseColor_Specified, specifiedColor=<rgba 0.0395 0.0395 0.0395 1>}
      uVRegion {bIsValid=false, max=<0 0>, min=<0 0>}
     }
     disabledForeground {colorUseRule=UseColor_ColorTable, specifiedColor=<rgba 0 0 0 0>}
     hovered {
      bIsDynamicallyLoaded=false
      drawAs=RoundedBox
      imageType=NoImage
      mirroring=NoMirror
      resourceName=None
      resourceObject=null
      tiling=NoTile
      imageSize=<32 32>
      margin=<ltrb 0 0 0 0>
      outlineSettings {
       bUseBrushTransparency=false
       roundingType=FixedRadius
       width=1
       color {colorUseRule=UseColor_ColorTable, specifiedColor=<rgba 0 0 0 0>}
       cornerRadii=<4 4 4 4>
      }
      tintColor {colorUseRule=UseColor_Specified, specifiedColor=<rgba 0.0953 0.0953 0.0953 1>}
      uVRegion {bIsValid=false, max=<0 0>, min=<0 0>}
     }
     hoveredForeground {colorUseRule=UseColor_ColorTable, specifiedColor=<rgba 0 0 0 0>}
     hoveredSlateSound {resourceObject=null}
     normal {
      bIsDynamicallyLoaded=false
      drawAs=RoundedBox
      imageType=NoImage
      mirroring=NoMirror
      resourceName=None
      resourceObject=null
      tiling=NoTile
      imageSize=<32 32>
      margin=<ltrb 0 0 0 0>
      outlineSettings {
       bUseBrushTransparency=false
       roundingType=FixedRadius
       width=1
       color {colorUseRule=UseColor_ColorTable, specifiedColor=<rgba 0 0 0 0>}
       cornerRadii=<4 4 4 4>
      }
      tintColor {colorUseRule=UseColor_Specified, specifiedColor=<rgba 0.0395 0.0395 0.0395 1>}
      uVRegion {bIsValid=false, max=<0 0>, min=<0 0>}
     }
     normalForeground {colorUseRule=UseColor_ColorTable, specifiedColor=<rgba 0 0 0 0>}
     normalPadding=<ltrb 12 1.5 12 1.5>
     pressed {
      bIsDynamicallyLoaded=false
      drawAs=RoundedBox
      imageType=NoImage
      mirroring=NoMirror
      resourceName=None
      resourceObject=null
      tiling=NoTile
      imageSize=<32 32>
      margin=<ltrb 0 0 0 0>
      outlineSettings {
       bUseBrushTransparency=false
       roundingType=FixedRadius
       width=1
       color {colorUseRule=UseColor_ColorTable, specifiedColor=<rgba 0 0 0 0>}
       cornerRadii=<4 4 4 4>
      }
      tintColor {colorUseRule=UseColor_Specified, specifiedColor=<rgba 0.0284 0.0284 0.0284 1>}
      uVRegion {bIsValid=false, max=<0 0>, min=<0 0>}
     }
     pressedForeground {colorUseRule=UseColor_ColorTable, specifiedColor=<rgba 0 0 0 0>}
     pressedPadding=<ltrb 12 2.5 12 0.5>
     pressedSlateSound {resourceObject=null}
    }}
   slot {class=HorizontalBoxSlot}
  }
  RichTextBlock_0 {
   widgetClassPath=/Script/UMG.RichTextBlock
   bIsVariable=true
   parent=CreateMontageButton
   widget=RichTextBlock_0
   properties {text=Create Harvester Tool Animation Montage, textStyleSet=/Game/Athena/HUD/EndGame/XpReasonStyleSheet}
   slot {class=ButtonSlot, properties {horizontalAlignment=HAlign_Left}}
  }
 ]
}
